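# Conversation database for the Temple NPC in the NPC library, entered through the Conversation.Entry.Temple tag.
# A single speech node gives a default farewell line, with contextual lines that can replace it.
# Prioritized requirement checks select a line for players who have visited The Spire or are talking to the NPC for the first time ever.
# The speech node then links out to the shared NPC library quests and basic services conversations.
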ConversationDatabase /NPCLibrary/NPCs/Temple/NPCLibrary_Conversation_Temple
entry_points [{tag=Conversation.Entry.Temple, destinations [#1]}]
nodes [
 ConversationEntryPointNode_7 {
  id=#1
  class=ConversationEntryPointNode
  outputs [#2]
  properties {entryTag {tagName=Conversation.Entry.Temple}}
 }
 Speech_C_4 {
  id=#2
  class=Speech_C
  outputs [#3, #4]
  properties {
   nodeName=""
   generalConfig {
    defaultMessage=Stay safe out there.
    contextualMessages [
     {
      message="Tell me, when you visited The Spire did it give you any... visions?\r\nFlashes of insight?"
      contextRequirements [{requirement {class=FortControllerRequirement_HasVisitedPOI, properties {pOITag {tagName=Athena.Location.POI.MainTower}}}}]
      priority {curve {rowName=Tandem.Speech.Priority.VisitedSpecificPOI}}
     }
     {
      message="I wish I had more information to share, but I am still attempting to figure this out myself."
      contextRequirements [{requirement {class=FortControllerRequirement_IsFirstEverConversationWithNPC}}]
      priority {curve {rowName=Tandem.Speech.Priority.FirstEverInteraction}}
     }
    ]
   }
  }
 }
 ConversationLinkNode_0 {id=#4, class=ConversationLinkNode, properties {remoteEntryTag {tagName=Conversation.Entry.BasicServices}}}
 ConversationLinkNode_1 {id=#3, class=ConversationLinkNode, properties {remoteEntryTag {tagName=Conversation.Entry.NPCLibrary.Quests}}}
]
